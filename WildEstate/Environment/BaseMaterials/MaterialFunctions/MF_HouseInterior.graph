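# Material function that fakes house interior shading by darkening the Base Color of incoming Material Attributes.
# Horizontal camera-to-pixel distance, scaled by two parameters, drives a gradual darkening with distance, up to a set limit.
# The flattened, normalized vertex normal is compared to world X, darkening faces that point away from that axis.
# Both factors multiply into Base Color; a Floor node on the distance is unconnected, so the falloff is not stepped.
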MaterialFunction /WildEstate/Environment/BaseMaterials/MaterialFunctions/MF_HouseInterior
graph {nodes [
  n1 = ComponentMask(None=CameraPositionWS()) {g=true, r=true}
  n2 = WorldPosition()
  n3 = Divide(
   A=Length(None=Subtract(A=@n1, B=@n2.XY)) {is=class MaterialExpressionLength}
   B=ScalarParameter() {defaultValue=1250, group=Interior Shading, parameterName=Interior Distance Darkening}
  )
  Floor(None=@n3)
  n4 = FunctionInput() {inputName=Material Attributes, inputType=FunctionInput_MaterialAttributes}
  n5 = GetMaterialAttributes(None=@n4) {attributeGetTypes [#1]}
  n6 = ComponentMask(None=VertexNormalWS()) {g=true, r=true}
  n7 = Constant()
  n8 = DotProduct(A=Normalize(VectorInput=AppendVector(A=@n6, B=@n7)), B=Constant3Vector() {constant=<rgba 1 0 0 0>})
  n9 = Saturate(None=Divide(A=@n3, B=ScalarParameter() {defaultValue=3, group=Interior Shading, parameterName=Interior Distance Levels}))
  n10 = Multiply(
   A=LinearInterpolate(A=ScalarParameter() {defaultValue=0.75, group=Interior Shading, parameterName=Max Normal Darkening}, Alpha=Abs(None=@n8)) {constA=0.5}
   B=LinearInterpolate(A=ScalarParameter() {defaultValue=0.5, group=Interior Shading, parameterName=Max Distance Darkening}, Alpha=OneMinus(None=@n9)) {constA=0.5}
  )
  FunctionOutput(None=SetMaterialAttributes!(MaterialAttributes=@n4, BaseColor=Multiply(A=@n5.BaseColor, B=@n10) {constB=0.5}) {attributeSetTypes [#1]})
 ]}
