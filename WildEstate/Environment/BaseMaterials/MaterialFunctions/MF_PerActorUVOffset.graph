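# Material function that shifts input UVs per instance so instances show different sections of a texture laid out as a grid.
# PerInstanceRandom is scaled by the NumberOfRows parameter and rounded into a U step, while a modulo-wrapped copy of it gives the V step.
# Both steps are divided by the row count and added to the UVs when the UsePerInstanceRandomUV static switch is on, otherwise UVs pass through.
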MaterialFunction /WildEstate/Environment/BaseMaterials/MaterialFunctions/MF_PerActorUVOffset
graph {
 properties {description=Offsets the UVs to us edifferent sections of textures}
 nodes [
  n1 = FunctionInput() {inputName=UVs, inputType=FunctionInput_Vector2}
  n2 = PerInstanceRandom()
  n3 = ScalarParameter() {defaultValue=2, group=Per Instance UV, parameterName=PerInstanceUVs.NumberOfRows}
  n4 = Multiply(A=@n2, B=@n3) {constB=4}
  n5 = Multiply(A=@n2, B=Multiply(A=@n3) {constB=2}) {constB=4}
  n6 = Constant() {r=2}
  n7 = Multiply(A=Divide(A=Modulo!(A=@n5, B=@n6), B=@n6), B=@n3) {constB=4}
  n8 = AppendVector(A=Divide(A=Round(None=@n4), B=@n3) {constB=4}, B=Divide(A=Round(None=@n7), B=@n3) {constB=4})
  FunctionOutput(None=StaticSwitchParameter(True=Add(A=@n1, B=@n8) {constB=0}, False=@n1) {desc=Offsets the UVs to us edifferent sections of textures, group=Per Instance UV, parameterName=UsePerInstanceRandomUV})
  MaterialFunctionCall!(Vector2=@n8, MaximumNumberOfDigits, UVs, DebugTextLocation RG_UpperRight BA_LowerLeft, Component Spacing) {materialFunction=/Engine/Functions/Engine_MaterialFunctions02/Utility/DebugFloat2Values}
 ]
}
