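# Material function for WildEstate environment materials that overrides Roughness on distant actors, named as support for distance-based actor outlines.
# Scaled actor world position is summed and wrapped with Frac into a pseudo-random per-actor value, remapped to a low-to-mid roughness range.
# Camera-relative actor distance minus DistanceThreshold, saturated, switches near-instantly from original to per-actor roughness.
# The UseDistanceBasedActorOutlines static switch selects this result or the untouched input attributes.
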MaterialFunction /WildEstate/Environment/BaseMaterials/MaterialFunctions/MF_WildEstate_DistanceBasedActorOutlines
graph {nodes [
  n1 = FunctionInput() {inputType=FunctionInput_MaterialAttributes}
  n2 = GetMaterialAttributes(None=@n1) {attributeGetTypes [#1]}
  MaterialExpressionMultiply_0 = Multiply(A=ActorPositionWS(), B=Constant3Vector() {constant=<rgba 0.0015 0.0029 0.0037 0>})
  n3 = Convert(None=@MaterialExpressionMultiply_0) {
   nodeName=Break Float3
   convertInputs [{
     type=Vector3
     expressionInput {
      expression=&self:MaterialExpressionMultiply_0
      inputName=None
      mask=0
      maskA=0
      maskB=0
      maskG=0
      maskR=0
      outputIndex=0
     }
    }]
   convertOutputs [{type=Scalar}, {type=Scalar}, {type=Scalar}]
  }
  n4 = Add(A=Add(A=@n3, B=@n3), B=@n3)
  n5 = Distance(A=Constant(), B=ActorPositionWS() {originType=CameraRelative})
  n6 = ScalarParameter() {defaultValue=8000, group=DistanceBasedActorOutlines, parameterName=DistanceThreshold}
  n7 = LinearInterpolate(A=@n2.Roughness, B=LinearInterpolate(Alpha=Frac(None=@n4)) {constA=0.01, constB=0.49}, Alpha=Saturate(None=Subtract(A=@n5, B=@n6)))
  FunctionOutput(None=StaticSwitchParameter(
    True=SetMaterialAttributes!(MaterialAttributes=@n2.MaterialAttributes, Roughness=@n7) {attributeSetTypes [#1]}
    False=@n1
   ) {group=DistanceBasedActorOutlines, parameterName=UseDistanceBasedActorOutlines, sortPriority=0})
 ]}
