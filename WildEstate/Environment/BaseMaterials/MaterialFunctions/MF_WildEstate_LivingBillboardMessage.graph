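# Material function for Wild Estate that overlays an animated "living billboard" message on incoming material attributes, toggled by the Use_LivingBillboardMessage static switch.
# A FlipBook image sequence, swapped for an override texture on Switch and Android/PC-Perf when enabled, supplies a mask.
# The mask picks flippable text and background colors for base color and optional emissive.
# A channel-selectable SFX mask input and time-of-day multiplier are also defined, alongside per-platform mip bias.
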MaterialFunction /WildEstate/Environment/BaseMaterials/MaterialFunctions/MF_WildEstate_LivingBillboardMessage
graph {nodes [
  n1 = NamedRerouteUsage()
  n2 = VectorParameter() {group=LivingBillboard, parameterName=LB_LivingBillboard_Text_BC, sortPriority=19, defaultValue=<rgba 0.02 0.02 0.02 0>}
  n3 = VectorParameter() {group=LivingBillboard, parameterName=LB_LivingBillboard_BG_BC, sortPriority=20, defaultValue=<rgba 0.88 0.8333 0.6216 1>}
  n4 = NamedRerouteUsage()
  n5 = LinearInterpolate(
   A=LinearInterpolate(
    A=LinearInterpolate(A=@n2.RGB, B=@n3.RGB, Alpha=@n4)
    B=LinearInterpolate(A=@n3.RGB, B=@n2.RGB, Alpha=@n4)
    Alpha=ScalarParameter() {group=LivingBillboard, parameterName=LB_BaseColor_FlipMask, sliderMax=1, sortPriority=25}
   )
   B=NamedRerouteUsage()
   Alpha=ScalarParameter() {group=LivingBillboard, parameterName=LB_BlendBillboardColorWithBase, sortPriority=6}
  )
  n6 = Constant()
  n7 = VectorParameter() {group=LivingBillboard, parameterName=LB_LivingBillboard_Text_Emissive, sortPriority=19}
  n8 = VectorParameter() {group=LivingBillboard, parameterName=LB_LivingBillboard_BG_Emissive, sortPriority=20, defaultValue=<rgba 0 0 0 1>}
  n9 = NamedRerouteUsage()
  n10 = Multiply(
   A=LinearInterpolate(
    A=LinearInterpolate(A=@n7.RGB, B=@n8.RGB, Alpha=@n9)
    B=LinearInterpolate(A=@n8.RGB, B=@n7.RGB, Alpha=@n9)
    Alpha=ScalarParameter() {group=LivingBillboard, parameterName=LB_Emissive_FlipMask, sliderMax=1, sortPriority=25}
   )
   B=NamedRerouteUsage()
  )
  n11 = MaterialFunctionCall(Emissive=@n10) {materialFunction=&MF_FX_EmissiveAdjustments}
  n12 = BlendMaterialAttributes(
   A=NamedRerouteUsage()
   B=StaticSwitchParameter(
    True=SetMaterialAttributes!(MaterialAttributes=@n1, BaseColor=@n5, EmissiveColor=@n6) {attributeSetTypes [#1, #2]}
    False=SetMaterialAttributes(MaterialAttributes=@n1, BaseColor=@n5, EmissiveColor=@n11.Result) {attributeSetTypes [#1, #2]}
   ) {desc="True = BC, False = BC + Emissive", group=LivingBillboard, parameterName=OnlyBaseColor, sortPriority=-100}
   Alpha=NamedRerouteUsage()
  )
  n13 = NamedRerouteUsage()
  BlendMaterialAttributes(A=@n12, B=@n13, Alpha=ScalarParameter() {group=LivingBillboard, parameterName=LB_ImageSequence_MOBILEEnabled, sortPriority=500})
  CollectionParameter() {collection=/WildEstate/TechArt/MPCs/MPC_WildEstate_TA, parameterName=LB_ImageSequence}
  FunctionOutput(None=StaticSwitchParameter(True=@n12, False=@n13) {group=Wild Estate Tech Art FX, parameterName=Use_LivingBillboardMessage, sortPriority=-100})
  Multiply(A=@n10, B=NamedRerouteUsage())
  n14 = ChannelMaskParameter(None=FunctionInput() {inputName=SFX Mask, sortPriority=1}) {group=LivingBillboard, maskChannel=Blue, parameterName=SFX_ChannelMask, sortPriority=-10, defaultValue=<rgba 0 0 1 0>}
  NamedRerouteDeclaration(None=@n14.R) {name=SFX Mask, nodeColor=<rgba 0 1 0.0156 1>}
  n15 = ScalarParameter() {group=LivingBillboard, parameterName=LB_ImageSequence_Speed, primitiveDataIndex=28, sortPriority=5}
  n16 = Multiply(A=@n15, B=Time())
  n17 = Ceil(None=ScalarParameter() {group=LivingBillboard, parameterName=LB_ImageSequence_AnimOnOff, primitiveDataIndex=28, sortPriority=3})
  n18 = Add(
   A=TextureCoordinate() {coordinateIndex=2}
   B=ComponentMask(None=VectorParameter() {group=LivingBillboard, parameterName=LB_UVOffset, defaultValue=<rgba 0 0 0 1>}.RGB) {g=true, r=true}
  )
  n19 = Constant()
  n20 = Constant() {r=2}
  n21 = Constant() {r=1}
  n22 = TextureSampleParameter2D(UVs=MaterialFunctionCall(
    "Animation  Phase (0-1)"=LinearInterpolate(A=@n15, B=@n16, Alpha=@n17)
    Number of Rows=ScalarParameter() {defaultValue=2, group=LivingBillboard, parameterName=LB_ImageSequence_RowAmount, sortPriority=9}
    Number of Columns=ScalarParameter() {defaultValue=2, group=LivingBillboard, parameterName=LB_ImageSequence_ColumnAmount, sortPriority=10}
    UVs=@n18
    MipBias/Level=MaterialFunctionCall("Xb1,PS4"=@n19, "PC-High&&Epic,XSX,PS5"=@n19, "PC-Perf,Android"=@n20, Switch=@n21, "PC-Low(NotPerf)&&Medium"=@n20) {materialFunction=&PlatformSwitch_1D}.Result
    "Use Mip Bias (T) Level (F)"=StaticBool() {value=true}
   ) {materialFunction=&FlipBook}.UVs) {
   group=LivingBillboard
   parameterName=LB_LivingBillboardMessage
   sortPriority=0
   texture=/Game/Global/Textures/Default/Blanks/T_Blank_LegalWhite_sRGBColor
  }
  n23 = VectorParameter() {group=LivingBillboard, parameterName=LB_MobileScale, sortPriority=400, defaultValue=<rgba 1 1 0 1>}
  n24 = TextureSampleParameter2D(UVs=Multiply(A=@n18, B=ComponentMask(None=@n23.RGB) {g=true, r=true})) {
   group=LivingBillboard
   parameterName=LB_LivingBillboard_MobileOverride
   sortPriority=500
   texture=&T_Blank_LegalWhite_sRGBColor
  }
  n25 = Constant()
  n26 = ScalarParameter() {group=LivingBillboard, parameterName=LB_ImageSequence_MOBILEEnabled, sortPriority=500}
  n27 = LinearInterpolate(
   A=@n22.R
   B=@n24.R
   Alpha=MaterialFunctionCall("Xb1,PS4"=@n25, "PC-High&&Epic,XSX,PS5"=@n25, "PC-Perf,Android"=@n26, Switch=@n26, "PC-Low(NotPerf)&&Medium"=@n25) {materialFunction=&PlatformSwitch_1D}.Result
  )
  NamedRerouteDeclaration(None=@n27) {name=ImageSequence, nodeColor=<rgba 0.8984 1 0 1>}
  n28 = GetMaterialAttributes(None=NamedRerouteDeclaration(None=FunctionInput() {inputType=FunctionInput_MaterialAttributes}) {name=InMat, nodeColor=<rgba 0.3359 0 1 1>}) {attributeGetTypes [#1]}
  NamedRerouteDeclaration(None=@n28.BaseColor) {name=InBC, nodeColor=<rgba 1 0 0.3047 1>}
  NamedRerouteDeclaration(None=Saturate(None=Desaturation(None=@n27))) {name=BB_BGMask, nodeColor=<rgba 1 0 0.2109 1>}
  n29 = VectorParameter() {
   group=LivingBillboard
   parameterName=LB_LivingBillboard_EmissiveTODMultiplier
   sortPriority=50
   channelNames {a=Night, b=Evening, g=Day, r=Morning}
   defaultValue=<rgba 0.5 0.25 0.5 1>
  }
  NamedRerouteDeclaration(None=MaterialFunctionCall(Value_Morning=@n29.Morning, Value_Day=@n29.Morning, Value_Evening=@n29.Morning, Value_Night=@n29.Morning) {materialFunction=&Scalar_Time_of_Day}.Result) {name=TOD, nodeColor=<rgba 0.9453 1 0 1>}
  NamedRerouteDeclaration(None=@n28.MaterialAttributes) {name=InMatAtt, nodeColor=<rgba 0.7109 1 0 1>}
  NamedRerouteDeclaration(None=Power(Base=@n27)) {name=BB_BGMaskRound, nodeColor=<rgba 1 0 0.2109 1>}
  NamedRerouteDeclaration(None=ScalarParameter() {defaultValue=1, group=LivingBillboard, parameterName=LB_ImageSequence_EmissiveIntensity, sortPriority=9}) {name=Brightness, nodeColor=<rgba 1 0.75 0 1>}
  NamedRerouteUsage()
  NamedRerouteUsage()
 ]}
