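# WildEstate material function that offsets vertices along the camera ray to shape fur silhouettes, added to the incoming material attributes' World Position Offset.
# A Fresnel-like mask raises the dot of vertex normal and view direction to a tunable falloff power.
# The mask blends separate front and back offset distances between camera-facing and edge vertices.
# The UseFurSilhouette static switch passes attributes through unchanged when off.
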MaterialFunction /WildEstate/Environment/BaseMaterials/MaterialFunctions/MF_WildEstate_FurSilhouette
graph {nodes [
  n1 = FunctionInput() {inputType=FunctionInput_MaterialAttributes}
  n2 = GetMaterialAttributes(None=@n1) {attributeGetTypes [#1]}
  n3 = Normalize(VectorInput=WorldPosition() {worldPositionShaderOffset=WPT_CameraRelativeNoOffsets}.XYZ)
  n4 = Power(
   Base=DotProduct(A=VertexNormalWS(), B=Multiply(A=@n3) {constB=-1})
   Exp=ScalarParameter() {defaultValue=0.08, group=FurSilhouette, parameterName=FurSilhouette_Edge_Falloff}
  )
  n5 = LinearInterpolate(B=ScalarParameter() {group=FurSilhouette, parameterName=FurSilhouette_Offset_Front}, Alpha=Saturate(None=@n4))
  n6 = ScalarParameter() {group=FurSilhouette, parameterName=FurSilhouette_Offset_Back}
  n7 = Saturate(None=OneMinus(None=@n4))
  n8 = Add(A=@n2.WorldPositionOffset, B=Multiply(A=@n3, B=LinearInterpolate(A=@n5, B=@n6, Alpha=@n7)))
  FunctionOutput(None=StaticSwitchParameter(
    True=SetMaterialAttributes!(MaterialAttributes=@n2.MaterialAttributes, WorldPositionOffset=@n8) {attributeSetTypes [#1]}
    False=@n1
   ) {group=FurSilhouette, parameterName=UseFurSilhouette})
 ]}
