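# Material function that animates the zipline motor by rotating the motor and two wheels through World Position Offset.
# Box masks on UV1 separate the parts, each rotating about its own parameterized pivot on a local X axis.
# Time and a speed parameter drive the angle, with a reversed, doubled angle for small wheels.
# Normals are rotated to match, and static switch UseZiplineMotorFX bypasses the effect.
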MaterialFunction /WildEstate/Environment/Props/Gameplay/Gameplay_WEZipline_A/Materials/MF_WEZipline_MotorWPO
graph {nodes [
  n1 = FunctionInput() {inputName=Material, inputType=FunctionInput_MaterialAttributes}
  n2 = GetMaterialAttributes(None=@n1) {attributeGetTypes [#1, #2]}
  n3 = MaterialFunctionCall(Rotation Axis=NamedRerouteUsage(), Rotation Angle=NamedRerouteUsage()) {materialFunction=&FixRotateAboutAxisNormals}
  n4 = MaterialFunctionCall(Rotation Axis=NamedRerouteUsage(), Rotation Angle=NamedRerouteUsage()) {materialFunction=&FixRotateAboutAxisNormals}
  n5 = NamedRerouteUsage()
  n6 = NamedRerouteUsage()
  n7 = NamedRerouteUsage()
  n8 = Max(A=Max(A=NamedRerouteUsage(), B=NamedRerouteUsage()), B=NamedRerouteUsage())
  n9 = LinearInterpolate(
   A=@n2.Normal
   B=LinearInterpolate(
    A=LinearInterpolate(A=@n3.New World Space Vertex Normal, B=@n4.New World Space Vertex Normal, Alpha=@n5) {desc=Normal}
    B=MaterialFunctionCall(Rotation Axis=@n6, Rotation Angle=@n7) {materialFunction=&FixRotateAboutAxisNormals}.New World Space Vertex Normal
    Alpha=NamedRerouteUsage()
   ) {desc=Normal}
   Alpha=@n8
  )
  n10 = VectorParameter() {group=Zipline, parameterName=ZiplineMotor.WheelB.Offset, defaultValue=<rgba 0 0 0 1>}
  n11 = VectorParameter() {group=Zipline, parameterName=ZiplineMotor.WheelA.Offset, defaultValue=<rgba 0 0 0 1>}
  n12 = LinearInterpolate(
   A=RotateAboutAxis(
    NormalizedRotationAxis=NamedRerouteUsage()
    RotationAngle=NamedRerouteUsage()
    PivotPoint=TransformPosition(None=@n10.RGB) {transformType=TRANSFORMPOSSOURCE_World}
    Position=WorldPosition().XYZ
   )
   B=RotateAboutAxis(
    NormalizedRotationAxis=NamedRerouteUsage()
    RotationAngle=NamedRerouteUsage()
    PivotPoint=TransformPosition(None=@n11.RGB) {transformType=TRANSFORMPOSSOURCE_World}
    Position=WorldPosition().XYZ
   )
   Alpha=NamedRerouteUsage()
  ) {desc=WPO}
  n13 = RotateAboutAxis(
   NormalizedRotationAxis=NamedRerouteUsage()
   RotationAngle=NamedRerouteUsage()
   PivotPoint=TransformPosition(None=VectorParameter() {group=Zipline, parameterName=ZiplineMotor.Offset, defaultValue=<rgba 0 0 -12.2 1>}.RGB) {transformType=TRANSFORMPOSSOURCE_World}
   Position=WorldPosition().XYZ
  )
  n14 = NamedRerouteUsage()
  n15 = LinearInterpolate(A=@n2.WorldPositionOffset, B=Add(A=@n2.WorldPositionOffset, B=LinearInterpolate(A=@n12, B=@n13, Alpha=@n14) {desc=WPO}), Alpha=@n8)
  FunctionOutput(None=StaticSwitchParameter(
    True=SetMaterialAttributes!(MaterialAttributes=@n2.MaterialAttributes, Normal=@n9, WorldPositionOffset=@n15) {attributeSetTypes [#1, #2]}
    False=@n1
   ) {group=Wild Estate Tech Art FX, parameterName=UseZiplineMotorFX, sortPriority=2})
  n16 = VectorParameter() {group=Zipline, parameterName=ZiplineMotor.Mask, channelNames {a=Bounds Y, b=Bounds X, g=Center Y, r=Center X}, defaultValue=<rgba 1 0 1 1>}
  NamedRerouteDeclaration(None=MaterialFunctionCall!(
    A=TextureCoordinate() {coordinateIndex=1}
    B=AppendVector(A=@n16.Center X, B=@n16.Center X)
    Bounds=AppendVector(A=@n16.Bounds X, B=@n16.Bounds X)
    Edge Falloff=Constant()
   ) {materialFunction=/Engine/Functions/Engine_MaterialFunctions01/Gradient/BoxMask-2D}.Result) {name=MotorMask, nodeColor=<rgba 1 0.5391 0 1>}
  n17 = VectorParameter() {
   group=Zipline
   parameterName=ZiplineMotor.WheelMaskA
   channelNames {a=Bounds Y, b=Bounds X, g=Center Y, r=Center X}
   defaultValue=<rgba 1 1 1 1>
  }
  NamedRerouteDeclaration(None=MaterialFunctionCall(
    A=TextureCoordinate() {coordinateIndex=1}
    B=AppendVector(A=@n17.Center X, B=@n17.Center X)
    Bounds=AppendVector(A=@n17.Bounds X, B=@n17.Bounds X)
    Edge Falloff=Constant()
   ) {materialFunction=&BoxMask-2D}.Result) {name=WheelMaskA, nodeColor=<rgba 0.2188 0 1 1>}
  n18 = VectorParameter() {
   group=Zipline
   parameterName=ZiplineMotor.WheelMaskB
   channelNames {a=Bounds Y, b=Bounds X, g=Center Y, r=Center X}
   defaultValue=<rgba 0 1 1 1>
  }
  NamedRerouteDeclaration(None=MaterialFunctionCall(
    A=TextureCoordinate() {coordinateIndex=1}
    B=AppendVector(A=@n18.Center X, B=@n18.Center X)
    Bounds=AppendVector(A=@n18.Bounds X, B=@n18.Bounds X)
    Edge Falloff=Constant()
   ) {materialFunction=&BoxMask-2D}.Result) {name=WheelMaskB, nodeColor=<rgba 0.2188 0 1 1>}
  NamedRerouteDeclaration(None=Transform(None=Constant3Vector() {constant=<rgba 1 0 0 0>}) {transformSourceType=TRANSFORMSOURCE_Local}) {name=Axis, nodeColor=<rgba 1 0 0.4219 1>}
  NamedRerouteDeclaration(None=Multiply(A=Time() {bOverride_Period=true, period=999}, B=ScalarParameter() {defaultValue=-1, group=Zipline, parameterName=ZiplineMotor.RotationSpeed})) {name=RotationAngle, nodeColor=<rgba 0 0.7422 1 1>}
  NamedRerouteDeclaration(None=Multiply(A=NamedRerouteUsage(), B=Constant() {r=-2})) {name=RotationAngleSmallWheel, nodeColor=<rgba 0.2656 1 0 1>}
  Normalize()
 ]}
