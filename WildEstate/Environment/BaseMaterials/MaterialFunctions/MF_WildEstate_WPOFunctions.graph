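# Material function that bundles Wild Estate's bespoke World Position Offset movement, described as unused by most assets.
# It threads one Material Attributes input through a serial chain: sway, zipline motor, new wind, fan spin, wind v2, cloth, then advanced WPO.
# Each stage is a separate material function feeding the next, with optional inputs such as masks, axes and cloth settings left unconnected.
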MaterialFunction /WildEstate/Environment/BaseMaterials/MaterialFunctions/MF_WildEstate_WPOFunctions
graph {
 properties {description="Used for bespoke movement that isn't used for most assest.  "}
 nodes [
  n1 = MaterialFunctionCall!(Material=MaterialFunctionCall!(In=FunctionInput() {inputType=FunctionInput_MaterialAttributes}) {materialFunction=/WildEstateTech/Environment/BaseMaterials/MaterialFunctions/MF_WE_WPOSway}.Result) {materialFunction=/WildEstate/Environment/Props/Gameplay/Gameplay_WEZipline_A/Materials/MF_WEZipline_MotorWPO}
  n2 = MaterialFunctionCall!(
   In=MaterialFunctionCall!(Material Attributes=MaterialFunctionCall!(In=@n1.Result) {materialFunction=/WildEstateTech/Environment/BaseMaterials/MaterialFunctions/MF_WE_WPO_WindNew}.Result) {materialFunction=/WildEstateTech/Environment/BaseMaterials/MaterialFunctions/MF_WE_FanSpin}.Result
   Custom Mask
   Custom Rotation Axis
  ) {materialFunction=/WildEstateTech/Environment/BaseMaterials/MaterialFunctions/MF_WE_WindWPO_v2}
  FunctionOutput(None=MaterialFunctionCall!(In=MaterialFunctionCall!(
     In=@n2.Result
     Pan Speed
     WPO_ClothOffsetDirection
     Pan Direction
     WPO_ClothOffsetIntensity
     "Motion Mask (Defaults to the Red Vertex Color)"
     WPO_ClothWPOWindTextureOffset
     UVs
     WPO_ClothNormalIntensity
    ) {materialFunction=/WildEstateTech/Environment/BaseMaterials/MaterialFunctions/MF_WildEstate_ClothWPO}.Result) {materialFunction=/WildEstateTech/Environment/BaseMaterials/MaterialFunctions/MF_WildEstate_AdvancedWPO}.Result)
 ]
}
