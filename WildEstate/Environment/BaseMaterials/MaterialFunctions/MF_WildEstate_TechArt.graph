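# MF_WildEstate_TechArt is a material function that collects special-case effects for one-off Wild Estate assets, each meant to be switched off by default.
# Material attributes pass in series through FurSilhouette, WPOFunctions, FlagDissolve, ZPOverlay and GagHighlights functions before output.
# An SFXMask input is declared as a named reroute.
# A GlowDonut caustics call branches off after WPOFunctions but does not reach the output, noted as removed for reload.
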MaterialFunction /WildEstate/Environment/BaseMaterials/MaterialFunctions/MF_WildEstate_TechArt
graph {
 properties {description="Function to hold unique and specific functionality for specific \"One off\" assets. Each contained function should be behind a switch, defaulted to off"}
 nodes [
  n1 = MaterialFunctionCall!(In=MaterialFunctionCall!(In=NamedRerouteUsage()) {materialFunction=&/MF_WildEstate_FurSilhouette}.Result) {materialFunction=&/MF_WildEstate_WPOFunctions}
  n2 = MaterialFunctionCall!(Material=@n1.Result) {materialFunction=/WildEstateTech/Environment/BaseMaterials/MaterialFunctions/MF_WildEstate_FlagDissolve}
  FunctionOutput(None=MaterialFunctionCall!(In=MaterialFunctionCall!(Material=@n2.Result) {materialFunction=/WildEstateTech/Environment/BaseMaterials/MaterialFunctions/MF_WildEstate_ZPOverlay}.Result) {materialFunction=&/MF_WildEstate_GagHighlights}.Result)
  MaterialFunctionCall!(Material=@n1.Result) {desc=removed for reload, materialFunction=/WildEstate/TechArt/GlowDonut/Meltdown/Functions/MF_GlowDonut_Caustics}
  NamedRerouteDeclaration(None=FunctionInput() {bUsePreviewValueAsDefault=true, inputName=SFXMask, sortPriority=1}) {name=SFX Mask, nodeColor=<rgba 1 0 0.5625 1>}
  NamedRerouteDeclaration(None=FunctionInput() {inputName=Material, inputType=FunctionInput_MaterialAttributes}) {name=Material In, nodeColor=<rgba 1 0 0.7266 1>}
 ]
}
