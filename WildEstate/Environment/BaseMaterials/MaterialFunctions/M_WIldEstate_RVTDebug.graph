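# Debug material for Wild Estate that samples the Apollo landscape runtime virtual textures, with no material output connections shown.
# World Z minus RVT height, divided by BlendDistance and saturated, forms a height blend alpha beside a red debug attribute set.
# RVT normals convert to tangent space, yet feed the grass function's world-space normal override.
# A green-channel lerp remaps RVT base color to fix a mismatch; sRGB conversion chains have no input.
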Material /WildEstate/Environment/BaseMaterials/MaterialFunctions/M_WIldEstate_RVTDebug
graph {
 properties {bAutomaticallySetUsageInEditor=false, usedShadingModels=Default Lit}
 nodes [
  n1 = RuntimeVirtualTextureSampleParameter(World Position=WorldPosition().XYZ) {
   bEnableFeedback=false
   materialType=WorldHeight
   parameterName=RVT_WorldHeight
   virtualTexture=/Game/Athena/Apollo/Maps/Landscape/VT/RVT_Apollo_Landscape_Height
  }
  n2 = Subtract(A=WorldPosition().Z, B=NamedRerouteDeclaration(None=@n1.WorldHeight) {name=RVT_WordPosZ, nodeColor=<rgba 0 0.7656 1 1>})
  n3 = Saturate(None=Divide(
    A=NamedRerouteDeclaration(None=@n2) {name=RVT_WordPosZbiased, nodeColor=<rgba 0 0.7656 1 1>}
    B=ScalarParameter() {defaultValue=16, parameterName=BlendDistance}
   ))
  BlendMaterialAttributes(
   A=SetMaterialAttributes!(MaterialAttributes, Emissive Color, Base Color=Constant3Vector() {constant=<rgba 1 0 0.0213 1>}) {attributeSetTypes [#1, #2]}
   Alpha=@n3
  )
  Constant()
  Constant() {r=1}
  n4 = Convert() {
   nodeName=Break Float3
   convertInputs [{type=Vector3, expressionInput {expression=null, inputName=None, mask=0, maskA=0, maskB=0, maskG=0, maskR=0, outputIndex=0}}]
   convertOutputs [{type=Scalar}, {type=Scalar}, {type=Scalar}]
  }
  MaterialFunctionCall(
   X=MaterialFunctionCall!(sRGB=@n4) {materialFunction=/Engine/Functions/Engine_MaterialFunctions01/ImageAdjustment/sRGBToLinear}.Linear
   Y=MaterialFunctionCall(sRGB=@n4) {materialFunction=&sRGBToLinear}.Linear
   Z=MaterialFunctionCall(sRGB=@n4) {materialFunction=&sRGBToLinear}.Linear
  ) {materialFunction=&MakeFloat3}
  n5 = Convert() {
   nodeName=Break Float3
   convertInputs [{type=Vector3, expressionInput {expression=null, inputName=None, mask=0, maskA=0, maskB=0, maskG=0, maskR=0, outputIndex=0}}]
   convertOutputs [{type=Scalar}, {type=Scalar}, {type=Scalar}]
  }
  MaterialFunctionCall(
   X=MaterialFunctionCall!(Linear=@n5) {materialFunction=/Engine/Functions/Engine_MaterialFunctions01/ImageAdjustment/LinearTosRGB}.sRGB
   Y=MaterialFunctionCall(Linear=@n5) {materialFunction=&LinearTosRGB}.sRGB
   Z=MaterialFunctionCall(Linear=@n5) {materialFunction=&LinearTosRGB}.sRGB
  ) {materialFunction=&MakeFloat3}
  n6 = RuntimeVirtualTextureSampleParameter(World Position=WorldPosition().XYZ) {
   bEnableFeedback=false
   materialType=BaseColor_Normal_Specular_YCoCg
   parameterName=RVT_Base
   virtualTexture=/Game/Athena/Apollo/Maps/Landscape/VT/RVT_Apollo_Landscape
  }
  n7 = Transform(None=NamedRerouteDeclaration(None=@n6.Normal) {name=RVT_Normal_WS, nodeColor=<rgba 0 0.7656 1 1>}) {transformSourceType=TRANSFORMSOURCE_World, transformType=TRANSFORM_Tangent}
  MaterialFunctionCall!(
   NormalOverrideWS=Normalize(VectorInput=NamedRerouteDeclaration(None=@n7) {name=RVT_Normal_TS, nodeColor=<rgba 0 0.7656 1 1>})
   UseCamera facing BildBoards
   UseLandscapeCoords
  ) {materialFunction=&/MF_WildEstate_Grass}
  MaterialExpressionNamedRerouteDeclaration_8 = NamedRerouteDeclaration(None=@n6.BaseColor) {name=RVT_BaseColor, nodeColor=<rgba 0 0.7656 1 1>}
  n8 = Convert(None=@MaterialExpressionNamedRerouteDeclaration_8) {
   nodeName=Break Float3
   convertInputs [{
     type=Vector3
     expressionInput {
      expression=&self:MaterialExpressionNamedRerouteDeclaration_8
      inputName=None
      mask=0
      maskA=0
      maskB=0
      maskG=0
      maskR=0
      outputIndex=0
     }
    }]
   convertOutputs [{type=Scalar}, {type=Scalar}, {type=Scalar}]
  }
  MaterialFunctionCall(
   X=@n8
   Y=LinearInterpolate(
    A=ScalarParameter() {defaultValue=-0.0088, desc=Fix RVT MissMatch, parameterName=Param2}
    B=ScalarParameter() {defaultValue=1.0462, desc=Fix RVT MissMatch}
    Alpha=@n8
   )
   Z=@n8
  ) {materialFunction=&MakeFloat3}
  NamedRerouteDeclaration(None=@n3) {name=RVT_HeightBlend, nodeColor=<rgba 0 0.7656 1 1>}
 ]
}
