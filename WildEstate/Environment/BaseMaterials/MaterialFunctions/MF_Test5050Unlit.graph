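# Test material function that flat-shades environment materials by splitting base color between lit color and emissive, both driven by time of day.
# An optional mask-blended directional path adds brightness compression, normal-based corner break and a sun-facing multiplier from FortniteMaterialParameters.
# MaterialProxyReplace supplies separate HLOD bake values, including emissive colorspace correction.
# A DONTTOUCHTHIS static switch gates directional shading and selects a second Enabled/Disabled attributes output.
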MaterialFunction /WildEstate/Environment/BaseMaterials/MaterialFunctions/MF_Test5050Unlit
graph {nodes [
  n1 = GetMaterialAttributes(None=FunctionInput() {inputType=FunctionInput_MaterialAttributes}) {attributeGetTypes [#1, #2]}
  n2 = Add(A=NamedRerouteUsage(), B=ScalarParameter() {defaultValue=0.025, parameterName=Base Color Add})
  n3 = Multiply(A=@n2, B=ScalarParameter() {defaultValue=0.8, parameterName=Base Color Mult}) {constB=0.2}
  n4 = Constant() {r=1}
  n5 = Saturate(None=Subtract(A=Desaturation(None=@n2, Fraction=@n4), B=ScalarParameter() {defaultValue=0.2, parameterName=Base Color Threshold}) {constB=0.45})
  n6 = ComponentMask(None=VertexNormalWS()) {r=true}
  n7 = Multiply(A=Add(A=Abs(None=@n6)) {constB=-0.5}, B=ScalarParameter() {defaultValue=0.2, parameterName=Corner Break}) {constB=0.12}
  n8 = Constant() {r=0.66}
  n9 = MaterialFunctionCall(Value_Morning=@n8, Value_Day=@n8, Value_Evening=@n8, Value_Night=Constant() {r=0.5}) {materialFunction=&Scalar_Time_of_Day}
  n10 = Multiply(A=Multiply(A=LinearInterpolate(A=@n2, B=@n3, Alpha=@n5), B=Add(A=@n7)), B=@n9.Result) {constB=0.7}
  n11 = CollectionParameter() {
   collection=&FortniteMaterialParameters
   parameterName=EnvironmentalEmissiveMult
  }
  n12 = VertexNormalWS()
  n13 = SkyAtmosphereLightDirection()
  n14 = Multiply(A=Add(A=DotProduct(A=@n12, B=@n13))) {constB=0.5}
  n15 = Multiply(A=@n10, B=LinearInterpolate(B=@n11, Alpha=Saturate(None=@n14)) {constA=1})
  n16 = NamedRerouteUsage()
  n17 = Add(A=Multiply(A=@n10, B=NamedRerouteUsage()) {constB=0.2}, B=NamedRerouteUsage())
  n18 = Multiply(A=MaterialFunctionCall!(Realtime=NamedRerouteUsage()) {desc=This is to correct emissive colorspace, materialFunction=/Game/Athena/HLODSetup/Materials/MF_HLODProxyColorSpaceCorrection}.Result) {constB=1.75}
  n19 = NamedRerouteUsage()
  n20 = Multiply(A=@n9.Result) {constB=1.14}
  n21 = SetMaterialAttributes!(
   MaterialAttributes=@n1.MaterialAttributes
   BaseColor=MaterialProxyReplace(Realtime=Multiply(A=@n19, B=@n20) {constB=0.8}, MaterialProxy=@n19)
   EmissiveColor=Multiply(A=@n19, B=NamedRerouteUsage()) {constB=0.2}
  ) {attributeSetTypes [#1, #2]}
  n22 = BlendMaterialAttributes(
   A=SetMaterialAttributes(
    MaterialAttributes=@n1.MaterialAttributes
    BaseColor=MaterialProxyReplace(Realtime=@n15, MaterialProxy=@n16)
    EmissiveColor=MaterialProxyReplace(Realtime=@n17, MaterialProxy=@n18) {desc="We dont need this for the bake because the HLOD will calculate \r\nby itself."}
   ) {attributeSetTypes [#1, #2]}
   B=@n21
   Alpha=FunctionInput() {bUsePreviewValueAsDefault=true, inputName=DirectionalShadingMask, inputType=FunctionInput_Scalar, sortPriority=51}
  )
  n23 = FunctionInput(Preview=StaticBool()) {bUsePreviewValueAsDefault=true, inputName=useDirectionalShading, inputType=FunctionInput_StaticBool, sortPriority=50}
  FunctionOutput(None=StaticSwitchParameter(True=StaticSwitch(True=@n22, False=@n21, Value=@n23), False=@n21) {defaultValue=true, parameterName=DONTTOUCHTHIS})
  FunctionOutput(None=StaticSwitchParameter(
    True=FunctionInput() {inputName=Enabled, inputType=FunctionInput_MaterialAttributes, sortPriority=10}
    False=FunctionInput() {inputName=Disabled, inputType=FunctionInput_MaterialAttributes, sortPriority=20}
   ) {defaultValue=true, parameterName=DONTTOUCHTHIS}) {outputName=EnabledDisabledOut, sortPriority=10}
  NamedRerouteDeclaration(None=@n1.EmissiveColor) {name=Emission, nodeColor=<rgba 1 0 0.3516 1>}
  n24 = VectorParameter() {group=FlatShading, parameterName=Test_Morning, defaultValue=<rgba 0.1709 0.1615 0.2135 1>}
  n25 = VectorParameter() {group=FlatShading, parameterName=Test_Day, defaultValue=<rgba 0.2292 0.2292 0.2292 1>}
  n26 = VectorParameter() {group=FlatShading, parameterName=Test_Evening, defaultValue=<rgba 0.1875 0.1443 0.1843 1>}
  n27 = VectorParameter() {group=FlatShading, parameterName=Test_Night, defaultValue=<rgba 0.21 0.2392 0.3021 1>}
  NamedRerouteDeclaration(None=MaterialProxyReplace(
    Realtime=MaterialFunctionCall(Color_Morning=@n24.RGB, Color_Day=@n25.RGB, Color_Evening=@n26.RGB, Color_Night=@n27.RGB) {materialFunction=&Color_Time_of_Day}.Result
    MaterialProxy=Constant3Vector() {constant=<rgba 1 1 1 0>}
   )) {name=Unlit Emission, nodeColor=<rgba 1 0.1172 0 1>}
  NamedRerouteDeclaration(None=@n1.BaseColor) {name=Base Color, nodeColor=<rgba 0 0.2031 1 1>}
 ]}
