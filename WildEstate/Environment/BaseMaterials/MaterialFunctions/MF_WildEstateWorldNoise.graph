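# Material function that supplies world-space XY noise masks for Wild Estate environment materials, apparently for grass shading and transitions.
# Grass noise is quantized into steps, remapped between darkness and brightness limits, then saturated.
# TransitionNoise inverts a remapped noise texture and TransitionNoise_2 passes Perlin noise through; both become flat constants on the mobile shading path.
# The normal-based slope influence sum is unconnected, so the normal input affects no output.
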MaterialFunction /WildEstate/Environment/BaseMaterials/MaterialFunctions/MF_WildEstateWorldNoise
graph {nodes [
  n1 = FunctionInput(Preview=VertexNormalWS()) {bUsePreviewValueAsDefault=true, inputName=NormalOverrideWS}
  n2 = Multiply(A=Abs(None=ComponentMask(None=@n1) {r=true}), B=ScalarParameter() {defaultValue=-0.7611, parameterName=Grass_Quanitize_SlopeInfluence_X})
  n3 = Multiply(A=Abs(None=ComponentMask(None=@n1) {g=true}), B=ScalarParameter() {defaultValue=0.2709, parameterName=Grass_Quanitize_SlopeInfluence_Y})
  n4 = ComponentMask(None=@n1) {b=true}
  n5 = ScalarParameter() {defaultValue=0.4752, parameterName=Grass_Quanitize_SlopeInfluence}
  n6 = TextureSample(UVs=Divide(A=WorldPosition().XY, B=ScalarParameter() {defaultValue=15445.1445, parameterName=Grass_Noise_Size})) {
   samplerSource=SSM_Wrap_WorldGroupSettings
   samplerType=SAMPLERTYPE_LinearGrayscale
   texture=/WildEstate/Environment/TestContent/T_WildEstate_Landscape_Noise_M
  }
  Add(A=Add(A=Add(A=@n2, B=@n3), B=Multiply(A=@n4, B=@n5)), B=@n6.R)
  Constant() {r=0.5}
  Constant()
  n7 = ScalarParameter() {defaultValue=8, parameterName=GrassNoise_Quanitize_Steps}
  n8 = Round(None=Multiply(A=@n6.R, B=@n7))
  n9 = LinearInterpolate(
   A=ScalarParameter() {defaultValue=0.53, group=QuantizeOverlay, parameterName=Grass_Noise_Quantize_MinDarkness}
   B=ScalarParameter() {defaultValue=1.05, group=QuantizeOverlay, parameterName=Grass_Noise_Quantize_MaxBrigthness}
   Alpha=Saturate(None=Divide(A=@n8, B=@n7))
  )
  FunctionOutput(None=NamedRerouteDeclaration(None=Saturate(None=@n9)) {name=Grass Noise, nodeColor=<rgba 1 0.6797 0 1>})
  n10 = Divide(A=WorldPosition().XY, B=ScalarParameter() {defaultValue=294.9043, group=Dirt, parameterName=SizeGrassTransitionNoise})
  n11 = TextureSample(UVs=@n10) {
   samplerSource=SSM_Wrap_WorldGroupSettings
   samplerType=SAMPLERTYPE_LinearGrayscale
   texture=/WildEstate/Environment/Textures/WildEstate_Landscape_TNoise_M
  }
  n12 = LinearInterpolate(
   A=ScalarParameter() {defaultValue=0.8, parameterName=TNoise_Min}
   B=ScalarParameter() {defaultValue=1.3093, parameterName=TNoise_Max}
   Alpha=@n11.R
  )
  FunctionOutput(None=ShadingPathSwitch(Default=OneMinus(None=@n12), Mobile=Constant() {r=0.5}) {desc=Migrated from FeatureLevelSwitch}) {outputName=TransitionNoise, sortPriority=5}
  n13 = TextureSample(UVs=Divide(A=WorldPosition().XY, B=ScalarParameter() {defaultValue=679.8232, group=Dirt, parameterName=TransitionNoise_Size})) {
   samplerSource=SSM_Wrap_WorldGroupSettings
   samplerType=SAMPLERTYPE_Grayscale
   texture=/WildEstate/Landscape/Textures/WildEstate_Landscape_Perlin_M
  }
  FunctionOutput(None=ShadingPathSwitch(Default=@n13.R, Mobile=Constant() {r=0.5}) {desc=Migrated from FeatureLevelSwitch}) {outputName=TransitionNoise_2, sortPriority=10}
  PixelNormalWS()
  TextureSample(UVs=@n10) {samplerSource=SSM_Wrap_WorldGroupSettings, texture=/WildEstate/Environment/TestContent/T_WildEstatePine_Noise_M}
  VertexNormalWS()
 ]}
