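# Material function for the WildEstate asphalt layer, blending a flat-colored asphalt over an incoming material by layer weight.
# LayerWeight minus a threshold is rounded and saturated into a hard-edged mask, which drives the blend and is also output.
# The same mask switches ArrayTextureUV to the asphalt overlay size and array index, packed into one vector.
# A separate StaticMesh output skips the blend, multiplying the asphalt color by a world-XY-projected detail texture.
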MaterialFunction /WildEstate/Environment/BaseMaterials/MaterialFunctions/MF_WildEstate_Asphalt
graph {nodes [
  Constant2Vector()
  n1 = ScalarParameter() {defaultValue=2000, group=Asphalt, parameterName=Asphalt_OverlayTexture_Size}
  Divide(A=WorldPosition().XY, B=@n1)
  n2 = FunctionInput(Preview=Constant()) {bUsePreviewValueAsDefault=true, inputName=LayerWeight, inputType=FunctionInput_Scalar, sortPriority=30}
  n3 = ScalarParameter() {defaultValue=0.2, group=Asphalt, parameterName=AsphaltThreshold}
  n4 = Saturate(None=Round(None=Subtract(A=@n2, B=@n3)))
  FunctionOutput(None=@n4) {outputName=Mask, sortPriority=30}
  n5 = AppendVector(A=@n1, B=@n1)
  n6 = Constant() {desc=ArrrayIndex ID, r=3}
  FunctionOutput(None=LinearInterpolate(A=FunctionInput() {bUsePreviewValueAsDefault=true, inputName=ArrayTextureUV, sortPriority=100}, B=AppendVector(A=@n5, B=@n6), Alpha=@n4)) {outputName=ArrayTextureUV, sortPriority=50}
  n7 = SetMaterialAttributes!(
   MaterialAttributes
   BaseColor=VectorParameter() {group=Asphalt, parameterName=AsphaltColor, defaultValue=<rgba 0.187 0.1725 0.1398 1>}.RGB
   Roughness=Constant() {desc=CellShadingValue, r=0.95}
   Specular=Constant()
  ) {attributeSetTypes [#1, #2, #3]}
  FunctionOutput(None=BlendMaterialAttributes(A=FunctionInput() {inputName=In Material, inputType=FunctionInput_MaterialAttributes}, B=@n7, Alpha=@n4)) {outputName=Out Material, sortPriority=1}
  n8 = GetMaterialAttributes(None=@n7) {attributeGetTypes [#1]}
  n9 = TextureSample(UVs=Divide(A=WorldPosition().XY, B=ScalarParameter() {defaultValue=800, group=DetailTextures, parameterName=Detail_OverlayTexture_Size})) {texture=/WildEstate/Environment/Textures/T_WildEstate_Detail_Asphalt_A_A_D}
  FunctionOutput(None=SetMaterialAttributes!(MaterialAttributes=@n8.MaterialAttributes, BaseColor=Multiply(A=@n8.BaseColor, B=@n9.R)) {attributeSetTypes [#1]}) {outputName=Out StaticMesh, sortPriority=10}
 ]}
