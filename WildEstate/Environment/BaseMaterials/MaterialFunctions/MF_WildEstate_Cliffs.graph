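# Material function for WildEstate cliffs that blends a stylized cliff layer over the input material, with extra mask and static mesh outputs.
# A slope mask from vertex normal Z and world noise is rounded to hard edges.
# Biome-masked colors, boosted by sun facing, multiply world-aligned or UV0 outline masks that switch near to far by depth.
# Optional normals, color quantizing and flat shading follow; older distance-field line and debug branches remain unconnected.
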MaterialFunction /WildEstate/Environment/BaseMaterials/MaterialFunctions/MF_WildEstate_Cliffs
graph {nodes [
  n1 = VertexNormalWS()
  n2 = ComponentMask(None=@n1) {b=true}
  n3 = Divide(A=WorldPosition().XY, B=ScalarParameter() {defaultValue=4280.0078, parameterName=CliffTexture_size})
  Add(
   A=@n2
   B=Multiply(
    A=TextureSample(UVs=@n3) {samplerType=SAMPLERTYPE_LinearGrayscale, texture=/Game/Athena/Helios/Maps/Landmarks/Glacier/Textures/T_Artemis_Cliff_01_H}.R
    B=ScalarParameter() {defaultValue=0.0016, parameterName=Slopeinflucence}
   )
  )
  n4 = NamedRerouteUsage()
  FunctionOutput(None=@n4) {outputName=Mask, sortPriority=30}
  FunctionOutput(None=NamedRerouteUsage()) {outputName=Mask_For_GrassSpawner, sortPriority=35}
  n5 = BlendMaterialAttributes(A=NamedRerouteUsage(), B=FunctionInput() {inputName=In Material, inputType=FunctionInput_MaterialAttributes}, Alpha=@n4)
  FunctionOutput(None=@n5) {outputName=Out Material, sortPriority=1}
  FunctionOutput(None=NamedRerouteUsage()) {outputName=Out For StaticMesh, sortPriority=2}
  n6 = Multiply(A=TextureCoordinate(), B=ScalarParameter() {defaultValue=1, group=Cliff, parameterName=Cliffs_Overlay_UV_Tilling})
  n7 = TextureObjectParameter() {
   parameterName=Cliff_Mask_Texture
   samplerType=SAMPLERTYPE_Masks
   texture=/WildEstate/Environment/Terrain/Cliffs/Textures/T_WildeEstate_Cliff_Gorge_A_D_Paul
   channelNames {a="", b=ColorVariationMask, g=OutlineFar, r=OutlineNear}
  }
  n8 = ScalarParameter() {defaultValue=1200, group=Cliff, parameterName=Cliffs_Overlay_Texture_Size}
  n9 = MaterialFunctionCall(
   TextureObject=@n7
   TextureSize=@n8
   ProjectionTransitionContrast=ScalarParameter() {defaultValue=200, group=Cliff, parameterName=Cliffs_Overlay_Texture_ProjcectionContrast}
  ) {materialFunction=&WorldAlignedTexture}
  n10 = StaticSwitchParameter(
   True=TextureSample(UVs=@n6, Tex=@n7) {samplerSource=SSM_Wrap_WorldGroupSettings}.RGB
   False=StaticSwitchParameter(True=@n9.XYZ Texture, False=@n9.XYZ Texture) {group=Cliff, parameterName="XY or XYZ projcetion(Trui is XYZ)"}
  ) {group=Cliff, parameterName=Cliff_Overlay_Use_UV0, sortPriority=33}
  n11 = ComponentMask(None=@n10) {b=true}
  FunctionOutput(None=@n11) {outputName=Dirt_Masks, sortPriority=147}
  n12 = LinearInterpolate(
   A=VectorParameter() {parameterName=CliffColor_BIomeA_2, defaultValue=<rgba 0.18 0.1518 0.1341 1>}.RGB
   B=VectorParameter() {parameterName=CliffColor_BIomeA_1, defaultValue=<rgba 0.16 0.1295 0.115 1>}.RGB
   Alpha=NamedRerouteUsage()
  )
  Multiply(A=ScalarParameter() {defaultValue=0.756, parameterName=BiomeABrightness}, B=@n12)
  n13 = StaticBool() {value=true}
  n14 = MaterialFunctionCall!(
   UseLandscapeCoords=NamedRerouteDeclaration(None=FunctionInput(Preview=@n13) {bUsePreviewValueAsDefault=true, inputName=UseLandscapeCoords, inputType=FunctionInput_StaticBool, sortPriority=10}) {name=IsLandscape, nodeColor=<rgba 0.125 0 1 1>}
   "UseProjection(translucent)"
  ) {materialFunction=/WildEstate/Landscape/MaterialFunctions/MF_WildEstate_BiomeMasks}
  n15 = MaterialFunctionCall(Float3=@n14.QuadA_RGBA) {materialFunction=&BreakOutFloat3Components}
  n16 = LinearInterpolate(
   A=LinearInterpolate(
    A=ScalarParameter() {defaultValue=1.4, parameterName=BiomeA Shadow Boost}
    B=ScalarParameter() {defaultValue=1.4, parameterName=BiomeB Shadow Boost}
    Alpha=@n15.R
   )
   B=ScalarParameter() {defaultValue=1.4, parameterName=BiomeC Shadow Boost}
   Alpha=@n15.G
  )
  n17 = ScalarParameter() {defaultValue=1.15, parameterName=BiomeD Shadow Boost}
  n18 = MaterialFunctionCall(Float3=@n14.QuadB_RGBA) {materialFunction=&BreakOutFloat3Components}
  n19 = LinearInterpolate(
   A=LinearInterpolate(
    A=ScalarParameter() {defaultValue=1, parameterName=BiomeA Light Boost}
    B=ScalarParameter() {defaultValue=0.544, parameterName=BiomeB Light Boost}
    Alpha=@n15.R
   )
   B=ScalarParameter() {defaultValue=1, parameterName=BiomeC Light Boost}
   Alpha=@n15.G
  )
  n20 = ScalarParameter() {defaultValue=0.9156, parameterName=BiomeD Light Boost}
  n21 = CollectionParameter() {
   collection=&FortniteMaterialParameters
   parameterName=SunlightDirectionalVector
  }
  n22 = DotProduct(A=ComponentMask(None=@n21) {b=true, g=true, r=true}, B=StaticSwitch(True=VertexNormalWS(), False=PixelNormalWS(), Value=NamedRerouteUsage()))
  n23 = LinearInterpolate(
   A=LinearInterpolate(
    A=VectorParameter() {parameterName=CliffColor_Default_2, defaultValue=<rgba 0.1354 0.1189 0.1026 1>}.RGB
    B=VectorParameter() {parameterName=CliffColor_Default_1, defaultValue=<rgba 0.1667 0.1464 0.1263 1>}.RGB
    Alpha=NamedRerouteUsage()
   )
   B=@n12
   Alpha=@n15.R
  )
  n24 = LinearInterpolate(
   A=VectorParameter() {parameterName=CliffColor_BIomeC_2, defaultValue=<rgba 0.2865 0.1623 0.0998 1>}.RGB
   B=VectorParameter() {parameterName=CliffColor_BIomeC_1, defaultValue=<rgba 0.2083 0.1181 0.0726 1>}.RGB
   Alpha=NamedRerouteUsage()
  )
  n25 = VectorParameter() {parameterName=CliffColor_BIomeD_2, defaultValue=<rgba 0.6615 0.2547 0.1191 1>}
  n26 = VectorParameter() {parameterName=CliffColor_BIomeD_1, defaultValue=<rgba 0.4896 0.178 0.0832 1>}
  n27 = NamedRerouteUsage()
  n28 = Multiply(
   A=LinearInterpolate(A=LinearInterpolate(A=@n16, B=@n17, Alpha=@n18.R), B=LinearInterpolate(A=@n19, B=@n20, Alpha=@n18.R), Alpha=Saturate(None=@n22))
   B=LinearInterpolate(A=LinearInterpolate(A=@n23, B=@n24, Alpha=@n15.G), B=LinearInterpolate(A=@n25.RGB, B=@n26.RGB, Alpha=@n27), Alpha=@n18.R)
  )
  n29 = Subtract(A=PixelDepth().R, B=ScalarParameter() {defaultValue=5000, group=Cliff, parameterName=Cliffs_OutlineSwitch_OffsetDistance})
  n30 = ScalarParameter() {defaultValue=5000, group=Cliff, parameterName=Cliffs_OutlineSwitch_BlendDistance}
  n31 = LinearInterpolate(A=ComponentMask(None=@n10) {r=true}, B=ComponentMask(None=@n10) {g=true}, Alpha=Saturate(None=Divide(A=@n29, B=@n30)))
  n32 = ShadingPathSwitch(
   Default=@n31
   Mobile=LinearInterpolate(B=@n31, Alpha=ScalarParameter() {defaultValue=0.25, parameterName=WildEstate_TextureLinesOpacity}) {constA=1}
  ) {desc=Migrated from FeatureLevelSwitch}
  n33 = SetMaterialAttributes!(
   MaterialAttributes
   BaseColor=StaticSwitchParameter(True=@n10, False=Multiply(A=@n28, B=@n32)) {group=Cliff, parameterName=Cliff_Use_Pure_Diffuse, sortPriority=34}
   Roughness=Constant() {desc=CellShadingValue, r=0.9}
   Specular=Constant()
  ) {attributeSetTypes [#1, #2, #3]}
  n34 = TextureObjectParameter() {
   group=Normals
   parameterName=Normals
   samplerType=SAMPLERTYPE_Normal
   texture=/WildEstate/Environment/Terrain/Cliffs/Textures/T_WildeEstate_Cliff_Gorge_A_N
  }
  n35 = MaterialFunctionCall(
   TextureObject=@n34
   TextureSize=@n8
   WorldSpace=StaticBool()
   ProjectionTransitionContrast=ScalarParameter() {defaultValue=2, group=Normals, parameterName=Cliffs_Normal_Texture_ProjcectionContrast}
  ) {materialFunction=&WorldAlignedNormal}
  n36 = MaterialFunctionCall(
   Normal=StaticSwitchParameter(
    True=TextureSample(UVs=@n6, Tex=@n34) {samplerType=SAMPLERTYPE_Normal}.RGB
    False=StaticSwitchParameter(True=@n35.XYZ Texture, False=@n35.XYZ Texture) {group=Cliff, parameterName="XY or XYZ projcetion(Trui is XYZ)"}
   ) {group=Cliff, parameterName=Cliff_Overlay_Use_UV0, sortPriority=33}
   Flatness=ScalarParameter() {defaultValue=0.5, group=Normals, parameterName=Cliffs_Normal_Flatten}
  ) {materialFunction=&FlattenNormal}
  n37 = StaticSwitchParameter(
   True=SetMaterialAttributes!(MaterialAttributes=@n33, Normal=Transform(None=@n36.Result)) {attributeSetTypes [#4]}
   False=@n33
  ) {defaultValue=true, group=Normals, parameterName=Use Normal Texture}
  NamedRerouteDeclaration(None=MaterialFunctionCall!(
    Material In=MaterialFunctionCall!(In=@n37) {materialFunction=&/M_WildEstate_QantizeColor}.Result
    UseFlatShading=StaticBoolParameter() {group=FlatShading, parameterName=UseFlatShading_All, sortPriority=1}
    IsWorldSpace=StaticBool() {value=true}
    OverrideUnlitShadinBaseColorDarkening=ScalarParameter() {defaultValue=0.6, group=FlatShading, parameterName=Cliffs_UnlitFlatShading_BaseColorBrightness}
    OverrideUnlitShadinBaseColorEmissive=ScalarParameter() {defaultValue=0.5, group=FlatShading, parameterName=Cliffs_UnlitFlatShading_Emissive}
   ) {materialFunction=&/MF_WildEstate_FlatShading}.Result) {name=Final Mat Out, nodeColor=<rgba 0.5 0 1 1>}
  n38 = MaterialFunctionCall!(NormalOverrideWS) {materialFunction=&/MF_WildEstateWorldNoise}
  n39 = ScalarParameter() {parameterName=SlopeMask_HeightBias_Steep}
  n40 = ScalarParameter() {defaultValue=0.1664, parameterName=SlopeMask_HeightBias_Flat}
  n41 = Saturate(None=Power(Base=@n2, Exp=ScalarParameter() {defaultValue=3.9108, parameterName=SlopeMask_GradientFalloth}))
  n42 = ScalarParameter() {defaultValue=0.1088, parameterName=SlopeMask_GradientLengthSteep}
  n43 = ScalarParameter() {defaultValue=0.0928, parameterName=SlopeMask_GradientLength}
  n44 = Divide(
   A=Subtract(A=@n2, B=ScalarParameter() {defaultValue=0.6602, parameterName=SlopeMask_Min})
   B=Max(A=LinearInterpolate(A=@n42, B=@n43, Alpha=@n41)) {constB=0}
  )
  n45 = Add(A=Multiply(A=OneMinus(None=@n38.TransitionNoise), B=LinearInterpolate(A=@n39, B=@n40, Alpha=@n41)), B=@n44)
  n46 = NamedRerouteDeclaration(None=@n45) {name=CliffMask_Original, nodeColor=<rgba 0.0547 0 1 1>}
  NamedRerouteDeclaration(None=Saturate(None=Round(None=@n46))) {name=CliffMask, nodeColor=<rgba 0.0547 0 1 1>}
  NamedRerouteDeclaration(None=@n37) {name=Final Mat Out_StaticMesh, nodeColor=<rgba 0.5 0 1 1>}
  NamedRerouteDeclaration(None=@n11) {name=TextureColorVarMask, nodeColor=<rgba 0 0.8594 1 1>}
  n47 = Subtract(
   A=Length(None=ComponentMask(None=@n1) {g=true, r=true}) {is=class MaterialExpressionLength}
   B=ScalarParameter() {defaultValue=0.704, parameterName=SlopeMask_GradientLength_bias}
  )
  n48 = ScalarParameter() {defaultValue=0.001, parameterName=SlopeMask_GradientLengthPow}
  NamedRerouteDeclaration(None=Saturate(None=Divide(A=@n47, B=@n48))) {name=DebugMask1, nodeColor=<rgba 1 0 0.375 1>}
  OneMinus(None=@n2)
  Round(None=@n22)
  n49 = VertexNormalWS()
  n50 = Constant3Vector() {constant=<rgba 1 1 0.01 0>}
  n51 = Multiply(A=WorldPosition() {worldPositionShaderOffset=WPT_CameraRelative}.XYZ) {constB=-1}
  n52 = ScalarParameter() {defaultValue=0.5947, parameterName=StepsRougnessCells}
  n53 = Multiply(A=DotProduct(A=Multiply(A=@n49, B=@n50), B=Normalize(VectorInput=@n51)), B=@n52)
  Saturate(None=Divide(A=Round(None=@n53), B=@n52))
  Saturate(None=@n44)
  ScalarParameter() {defaultValue=20000, parameterName=LineFadeDistance}
  ScalarParameter() {defaultValue=0.2344, parameterName=CliffSlopeMask_Min}
  n54 = MaterialFunctionCall(
   TextureObject=TextureObject() {samplerType=SAMPLERTYPE_Alpha, texture=/WildEstate/Environment/TestContent/T_WildEstate_Cliff_DF_D}
   TextureSize=ScalarParameter() {defaultValue=4280.0078, parameterName=CliffTexture_Size}
   ProjectionTransitionContrast=ScalarParameter() {defaultValue=24250.2227, parameterName=CliffProjectionContrast}
  ) {materialFunction=&WorldAlignedTexture}
  n55 = Multiply(
   A=MaterialFunctionCall!(
    InDistanceFieldTexture=ComponentMask(None=@n54.XY Texture) {r=true}
    LineWidthNear=ScalarParameter() {defaultValue=0.3917, parameterName=LineWidthNear}
    LineWidthFar=ScalarParameter() {defaultValue=0.4875, parameterName=LineWidthFar}
    LineTransitionDistance=ScalarParameter() {defaultValue=50000, parameterName=LineWidthTransitionDistance}
    LineTransitionOffset=ScalarParameter() {defaultValue=2000, parameterName=LineFadeOffset}
    LineFadeOpacity=ScalarParameter() {defaultValue=0.35, parameterName=LineFadeOpacity}
    IsEnabled=StaticBoolParameter() {parameterName=UseDistanceFieldOverlayTexture}
   ) {materialFunction=&/MF_WildEstate_DistanceFieldLineThickness}.Switch_0
   B=VectorParameter() {parameterName=CliffColor, defaultValue=<rgba 0.2865 0.2479 0.2208 1>}.RGB
  )
  n56 = PixelDepth()
  n57 = ScalarParameter() {defaultValue=1336565.5, parameterName=SlopeMask_DistanceAdd}
  n58 = Divide(
   A=OneMinus(None=@n45)
   B=LinearInterpolate(
    A=ScalarParameter() {defaultValue=1.0008, parameterName=SlopeMask_LineWidth}
    B=ScalarParameter() {defaultValue=1.095, parameterName=SlopeMask_LineWidth2}
    Alpha=Divide(A=@n56.R, B=@n57) {constB=50000}
   )
  )
  n59 = Multiply(A=@n55, B=Saturate(None=Round(None=@n58)))
  SetMaterialAttributes!(
   MaterialAttributes
   BaseColor=Multiply(
    A=MaterialFunctionCall() {materialFunction=&MF_WildEstateWorldNoise}.Result
    B=ShadingPathSwitch(Default=@n55, Mobile=@n59) {desc=Migrated from FeatureLevelSwitch}
   )
   Roughness=Constant() {desc=CellShadingValue, r=0.4}
   Specular=Constant()
  ) {attributeSetTypes [#1, #2, #3]}
  n60 = GetMaterialAttributes(None=@n5) {attributeGetTypes [#1]}
  n61 = NamedRerouteUsage()
  n62 = Constant3Vector() {constant=<rgba 0 0 1 0>}
  SetMaterialAttributes!(MaterialAttributes=@n60.MaterialAttributes, BaseColor=Add(A=@n60.BaseColor, B=Multiply(A=@n61, B=@n62))) {attributeSetTypes [#1]}
 ]}
