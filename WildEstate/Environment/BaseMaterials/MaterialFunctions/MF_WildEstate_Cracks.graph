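# Material function that overlays damage cracks on Wild Estate destructibles, driven by a gameplay-set destruction amount (custom primitive data by default).
# A crack texture, projected from pre-skinned local position and normal, blends its R, G and B channels as damage rises.
# A noise-distorted sphere mask varies early stages, with noise sampling gated on nonzero damage.
# The mask darkens opaque base color and emissive; its inverse raises translucent emissive and opacity.
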MaterialFunction /WildEstate/Environment/BaseMaterials/MaterialFunctions/MF_WildEstate_Cracks
graph {nodes [
  Constant() {r=1}
  n1 = GetMaterialAttributes(None=FunctionInput() {inputType=FunctionInput_MaterialAttributes}) {attributeGetTypes [#1, #2]}
  n2 = NamedRerouteUsage()
  FunctionOutput(None=SetMaterialAttributes!(MaterialAttributes=@n1.MaterialAttributes, BaseColor=Multiply(A=@n1.BaseColor, B=@n2), EmissiveColor=Multiply(A=@n1.EmissiveColor, B=@n2)) {attributeSetTypes [#1, #2]})
  n3 = GetMaterialAttributes(None=FunctionInput() {inputName=In_Translucent, inputType=FunctionInput_MaterialAttributes, sortPriority=10}) {attributeGetTypes [#2, #3]}
  n4 = OneMinus(None=NamedRerouteUsage())
  n5 = Multiply(A=@n4) {constB=0.1}
  n6 = Add(A=@n3.Opacity, B=@n4)
  FunctionOutput(None=SetMaterialAttributes!(MaterialAttributes=@n3.MaterialAttributes, EmissiveColor=Add(A=@n3.EmissiveColor, B=@n5), Opacity=Saturate(None=@n6)) {attributeSetTypes [#2, #3]}) {outputName=Result_Translucent, sortPriority=10}
  n7 = Constant() {r=0.1}
  n8 = ComponentMask(None=NamedRerouteUsage()) {b=true}
  n9 = Constant() {r=1.5}
  n10 = Multiply(A=Subtract(A=NamedRerouteUsage(), B=Constant() {r=0.6}) {constB=0.1}, B=Constant() {r=4.6}) {constB=3}
  n11 = Saturate(None=@n10)
  n12 = NamedRerouteUsage()
  n13 = Multiply(A=@n12, B=Constant() {r=3}) {constB=2}
  n14 = Multiply(A=ComponentMask(None=NamedRerouteUsage()) {r=true}, B=LinearInterpolate(Alpha=Saturate(None=@n13)) {constA=4, constB=1.2}) {constB=10}
  n15 = Constant() {r=1.4}
  n16 = Multiply(A=Subtract(A=NamedRerouteUsage(), B=Constant() {r=0.35}) {constB=0.1}, B=Constant() {r=4}) {constB=3}
  n17 = Multiply(A=ComponentMask(None=NamedRerouteUsage()) {g=true}, B=LinearInterpolate(A=@n15, B=@n15, Alpha=Saturate(None=@n16)) {constA=2, constB=1.2})
  n18 = Saturate(None=@n16)
  n19 = TextureObject() {texture=/Game/TimeOfDay/LightFunction/T_Cloud2dNoise}
  n20 = TextureCoordinate() {uTiling=0.3, vTiling=0.3}
  n21 = StaticSwitch(
   True=MaterialFunctionCall!(Texture=@n19, UVs=@n20, bDoWork=NamedRerouteUsage()) {materialFunction=/Game/Packages/jl_tests/CrackMaterial/MF_ConditionalColorTextureSample}.RGB
   False=TextureSample(UVs=@n20, Tex=@n19) {samplerSource=SSM_Wrap_WorldGroupSettings, texture=&T_Cloud2dNoise}.RGB
  ) {defaultValue=true}
  n22 = Add(A=WorldPosition().XYZ, B=Multiply(A=Power(Base=@n21)) {constB=160})
  n23 = Add(A=ObjectPositionWS(), B=Constant3Vector() {constant=<rgba 60 0 20 0>})
  n24 = LinearInterpolate(Alpha=Multiply(A=@n12, B=Constant() {r=2.3}) {constB=2}) {constA=600, constB=0}
  n25 = Constant() {r=0.1}
  n26 = LinearInterpolate(
   A=Add(
    A=LinearInterpolate(A=@n14, B=@n17, Alpha=@n18)
    B=SphereMask(A=@n22, B=@n23, Radius=@n24, Hardness=@n25) {attenuationRadius=512, hardnessPercent=50}
   )
   B=Multiply(A=Add(A=@n8) {constB=0.05}, B=LinearInterpolate(A=@n9, B=@n9, Alpha=@n11) {constA=2, constB=1.2})
   Alpha=Saturate(None=@n10)
  )
  n27 = Multiply(A=NamedRerouteUsage(), B=Constant() {r=5}) {constB=3}
  n28 = Add(
   A=Add(
    A=Saturate(None=@n26)
    B=ScalarParameter() {defaultValue=0.05, group=z_GameplayParameters_DestructibleDamage, parameterName=FN_DestructibeDamage_CrackColorBrightness}
   )
   B=Saturate(None=OneMinus(None=@n27))
  )
  NamedRerouteDeclaration(None=Saturate(None=Add(A=@n7, B=@n28))) {name=CrackAni, nodeColor=<rgba 1 0.7969 0 1>}
  n29 = ScalarParameter() {defaultValue=700, group=z_GameplayParameters_DestructibleDamage, parameterName=Destructible Damage Cracks Size, sortPriority=320}
  n30 = LocalPosition() {localOrigin=InstancePreSkinning}
  n31 = PreSkinnedNormal()
  NamedRerouteDeclaration(None=MaterialFunctionCall(
    TextureObject=TextureObject() {texture=/WildEstate/Environment/Textures/T_Wildestate_DamageCracks}
    TextureSize=@n29
    WorldPosition=VertexInterpolator(VS=@n30.XYZ).PS
    World Space Normal=VertexInterpolator(VS=@n31).PS
   ) {materialFunction=&WorldAlignedTexture}.XYZ Texture) {name=CRACK, nodeColor=<rgba 0.6875 0 1 1>}
  n32 = ScalarParameter() {
   bUseCustomPrimitiveData=true
   desc="Destructible Damage. This parameter is controlled by gameplay code.\r\nDefault=Custom Primitive Data Index 13"
   group=z_GameplayParameters_DestructibleDamage
   parameterName=Code_DestructionAmount
   primitiveDataIndex=13
   sortPriority=105
  }
  n33 = FunctionInput(Preview=@n32) {
   bUsePreviewValueAsDefault=true
   description=Default parameter is controlled by code. Custom Primitive Data Index 13
   inputName=DestructibleDamageAmount
   inputType=FunctionInput_Scalar
   sortPriority=200
   previewValue=<0 1 0 1>
  }
  n34 = Constant()
  NamedRerouteDeclaration(None=If(
    A=NamedRerouteDeclaration(None=@n33) {name=DestructionAmount, nodeColor=<rgba 1 0 0.7031 1>}
    "A > B"=Constant() {r=1}
    "A == B"=@n34
    "A < B"=@n34
   )) {name=bEnableTextureSample, nodeColor=<rgba 0.6875 1 0 1>}
  ScalarParameter() {defaultValue=0.6, parameterName=Test Cracks}
 ]}
