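# Material function that outputs a close-up grass noise multiplier for WildEstate landscape materials, fading out with distance.
# World XY, scaled by a size parameter, samples the grass overlay texture's green channel to blend from an intensity value toward one.
# Pixel depth, offset and divided by a fade distance, blends that result to neutral.
# A ShadingPathSwitch with a constant mobile value exists but does not feed the output.
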MaterialFunction /WildEstate/Environment/BaseMaterials/MaterialFunctions/MF_WildEstate_LandscapeCloseUpTexture
graph {nodes [
  n1 = Divide(A=WorldPosition().XY, B=ScalarParameter() {defaultValue=1100, group=Grass, parameterName=CloseUpGrassNoise_Size})
  n2 = Subtract(A=PixelDepth().R, B=ScalarParameter() {defaultValue=1000, group=Grass, parameterName=CloseUpGrassNoise_offset})
  n3 = ScalarParameter() {defaultValue=6535.71, group=Grass, parameterName=CloseUpGrassNoise_FadeDist}
  n4 = LinearInterpolate(
   A=LinearInterpolate(
    A=ScalarParameter() {defaultValue=0.9, group=Grass, parameterName=CloseUpGrassNoise_Intensity}
    Alpha=TextureSample(UVs=@n1) {samplerSource=SSM_Wrap_WorldGroupSettings, texture=/WildEstate/Landscape/Textures/WildEstate_Landscape_Grass_Overlay_D}.G
   )
   Alpha=Saturate(None=Divide(A=@n2, B=@n3))
  )
  FunctionOutput(None=@n4) {outputName=Out Material, sortPriority=10}
  ShadingPathSwitch(Default=@n4, Mobile=Constant() {r=1}) {desc=Migrated from FeatureLevelSwitch}
 ]}
