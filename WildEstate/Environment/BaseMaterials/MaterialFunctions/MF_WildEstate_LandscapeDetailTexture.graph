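# Landscape material function that overlays detail textures on incoming material attributes for WildEstate.
# Skewed world XY, scaled by ArrayTextureUV's RG with B as slice, samples an overlay array whose red channel multiplies BaseColor.
# A separate output provides a world-projected grass overlay's red channel for static meshes.
# The CliffMask input, Rotator, mobile shading-path switch and extra grass and concrete samples feed no output.
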MaterialFunction /WildEstate/Environment/BaseMaterials/MaterialFunctions/MF_WildEstate_LandscapeDetailTexture
graph {nodes [
  FunctionInput() {inputName=CliffMask, inputType=FunctionInput_Scalar, sortPriority=200}
  n1 = WorldPosition()
  n2 = ScalarParameter() {defaultValue=800, group=DetailTextures, parameterName=Detail_OverlayTexture_Size}
  FunctionOutput(None=TextureSample(UVs=Divide(A=@n1.XY, B=@n2)) {samplerSource=SSM_Wrap_WorldGroupSettings, texture=/WildEstate/Landscape/Textures/WildEstate_Landscape_Grass_Overlay_D}.R) {outputName=GrassDetailTextureForStaticMesh, sortPriority=100}
  n3 = FunctionInput() {inputName=In Material, inputType=FunctionInput_MaterialAttributes}
  n4 = GetMaterialAttributes(None=@n3) {attributeGetTypes [#1]}
  n5 = Multiply(
   A=MaterialFunctionCall(XYZ=@n1.XY, XY=@n1.XY) {materialFunction=&Swizzle}.YX
   B=ScalarParameter() {defaultValue=0.15, parameterName=Skew}
  )
  n6 = FunctionInput() {bUsePreviewValueAsDefault=true, inputName=ArrayTextureUV, sortPriority=100}
  n7 = AppendVector(A=Divide(A=Add(A=@n5, B=@n1.XY), B=ComponentMask(None=@n6) {g=true, r=true}), B=ComponentMask(None=@n6) {b=true})
  n8 = SetMaterialAttributes!(
   MaterialAttributes=@n4.MaterialAttributes
   BaseColor=Multiply(
    A=TextureSample(UVs=@n7) {samplerType=SAMPLERTYPE_Masks, texture=/WildEstate/Landscape/Textures/WildEstate_Lanscape_OverlayArray}.R
    B=@n4.BaseColor
   )
  ) {attributeSetTypes [#1]}
  FunctionOutput(None=@n8) {outputName=Out Material, sortPriority=10}
  Rotator(Coordinate=@n1.XY, Time=ScalarParameter() {defaultValue=-2.0157, parameterName=Rotation})
  ShadingPathSwitch(Default=@n8, Mobile=@n3) {desc=Migrated from FeatureLevelSwitch}
  n9 = Divide(A=WorldPosition().XY, B=ScalarParameter() {defaultValue=1034.8248, group=DetailTextures, parameterName=DetailTextureSize})
  TextureSample(UVs=@n9) {texture=&WildEstate_Landscape_Grass_Overlay_D}
  TextureSample(UVs=@n9) {texture=/WildEstate/Environment/Textures/T_WildEstate_Detail_Concrete_A_A_D}
 ]}
