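# WildEstate material function that blends grass color, roughness and an up or RVT normal onto input material attributes near the base of objects.
# Height comes from RVT world Z, or instance pivot or bounds without RVT, driving a side-projected bottom-edge mask texture.
# Layer visibility and RVT specular can gate the mask, which fades with depth on mobile.
# Outputs blended attributes, bypassed in material proxies, plus Grass Mask.
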MaterialFunction /WildEstate/Environment/BaseMaterials/MaterialFunctions/M_WildEstate_GrassBlend
graph {nodes [
  ComponentMask(None=ObjectBounds()) {b=true}
  n1 = FunctionInput() {inputType=FunctionInput_MaterialAttributes}
  n2 = Constant() {r=1}
  n3 = StaticBool()
  n4 = MaterialFunctionCall!(NormalOverrideWS, UseCamera facing BildBoards=@n3, UseLandscapeCoords=@n3) {materialFunction=&/MF_WildEstate_Grass}
  n5 = Multiply(
   A=FunctionInput() {bUsePreviewValueAsDefault=true, inputName=RVT_BaseColor, sortPriority=20}
   B=MaterialFunctionCall() {materialFunction=&/MF_WildEstate_LandscapeCloseUpTexture}.Out Material
  )
  n6 = LinearInterpolate(A=@n5, B=@n4.Grass BaseColor, Alpha=NamedRerouteUsage())
  n7 = MaterialFunctionCall!(
   NoRVT=@n4.Grass BaseColor
   RVT=StaticSwitchParameter(True=StaticSwitchParameter(True=@n6, False=@n5) {group=Grass Blend, parameterName=UseAutoGrassblend}, False=@n4.Grass BaseColor) {group=Grass Blend, parameterName="Use RVT  GrassBlend"}
  ) {materialFunction=/Game/Packages/Fortress_SharedMaterials/Base_Material_Functions/FN_RVT_Switch_Color}
  n8 = Constant3Vector() {constant=<rgba 0 0 1 0>}
  n9 = StaticSwitch(
   True=@n8
   False=Transform(None=@n8) {transformSourceType=TRANSFORMSOURCE_World, transformType=TRANSFORM_Tangent}
   Value=FunctionInput(Preview=StaticBool()) {bUsePreviewValueAsDefault=true, inputName=IsWorldSpace, inputType=FunctionInput_StaticBool, sortPriority=20}
  )
  n10 = FunctionInput() {bUsePreviewValueAsDefault=true, inputName=RVT_Normal, sortPriority=20}
  n11 = StaticSwitchParameter(True=LinearInterpolate(A=@n10, B=@n9, Alpha=NamedRerouteUsage()), False=@n10) {group=Grass Blend, parameterName=UseAutoGrassblend}
  n12 = SetMaterialAttributes!(
   MaterialAttributes
   BaseColor=ShadingPathSwitch(Default=@n7.Result, Mobile=Multiply(A=@n7.Result, B=@n2)) {desc=Migrated from FeatureLevelSwitch}
   Normal=MaterialFunctionCall(NoRVT=@n9, RVT=StaticSwitchParameter(True=@n11, False=@n9) {group=Grass Blend, parameterName="Use RVT  GrassBlend"}) {materialFunction=&FN_RVT_Switch_Color}.Result
   Roughness=MaterialFunctionCall() {materialFunction=/WildEstate/Foliage/BaseMaterial/MF_WildEstate_GrassRoughness}.Result
   Specular=Constant()
  ) {attributeSetTypes [
    #1
    #2
    #3
    #4
   ]}
  n13 = FunctionInput() {bUsePreviewValueAsDefault=true, inputName=RVT_Specular, inputType=FunctionInput_Scalar, sortPriority=21}
  n14 = Saturate(None=@n13)
  n15 = ScalarParameter() {defaultValue=88.4498, group=Grass Blend, parameterName=GrassBlendTextureSize}
  n16 = Multiply(A=@n15) {constB=4}
  n17 = AppendVector(A=AppendVector(A=@n16, B=@n16), B=@n15)
  n18 = WorldPosition()
  n19 = TransformPosition(None=Constant3Vector()) {transformSourceType=TRANSFORMPOSSOURCE_Instance, transformType=TRANSFORMPOSSOURCE_World}
  n20 = Subtract(A=ComponentMask(None=@n19) {b=true}, B=ScalarParameter() {group=Grass Blend, parameterName=GrassBlend_ZPosition_NoRVTAvailable})
  n21 = MaterialFunctionCall() {materialFunction=&ObjectScale}
  n22 = Add(
   A=ComponentMask(None=Bounds() {type=MEILB_PreSkinnedLocal}.Min) {b=true}
   B=ScalarParameter() {group=Grass Blend, parameterName=GrassBlend_ZPosition_OffsetNoRVT}
  )
  n23 = StaticSwitch(
   True=Subtract(A=@n18.XYZ, B=@n20)
   False=Subtract(A=Subtract(A=@n18.XYZ, B=@n19), B=Multiply(A=@n21.Scale Z, B=@n22))
   Value=FunctionInput(Preview=StaticBool()) {bUsePreviewValueAsDefault=true, inputName=IsTree, inputType=FunctionInput_StaticBool, sortPriority=500}
  )
  n24 = WorldPosition()
  n25 = FunctionInput(Preview=Constant()) {bUsePreviewValueAsDefault=true, inputName=RVT_WorldPosZ, inputType=FunctionInput_Scalar, sortPriority=20}
  n26 = Subtract(A=@n24.Z, B=@n25)
  n27 = Subtract(A=@n26, B=Max(A=Subtract(A=@n20, B=@n25), B=Constant()))
  n28 = StaticSwitchParameter(
   True=StaticSwitchParameter(True=AppendVector(A=@n24.XY, B=@n27), False=AppendVector(A=@n24.XY, B=@n26)) {group=Grass Blend, parameterName=UseAutoGrassblend}
   False=@n23
  ) {group=Grass Blend, parameterName="Use RVT  GrassBlend"}
  n29 = Divide(
   A=Subtract(
    A=MaterialFunctionCall(NoRVT=@n23, RVT=@n28) {materialFunction=&FN_RVT_Switch_Color}.Result
    B=ScalarParameter() {defaultValue=53.9642, group=Grass Blend, parameterName=GrassBlend_ZPosition}
   )
   B=Multiply(A=Abs(None=@n17)) {constB=-1}
  )
  n30 = AppendVector(A=ComponentMask(None=@n29) {g=true, r=true}, B=Clamp(None=ComponentMask(None=@n29) {b=true}) {maxDefault=0.9, minDefault=0.1})
  n31 = ComponentMask(None=@n30) {b=true, r=true}
  n32 = TextureObject() {samplerType=SAMPLERTYPE_Masks, texture=/WildEstate/Environment/Textures/T_GrassBottom_Blend_M}
  n33 = ComponentMask(None=@n30) {b=true, g=true}
  n34 = ComponentMask(None=VertexNormalWS()) {r=true}
  n35 = ComponentMask(None=MaterialFunctionCall(In=Abs(None=@n34), Contrast=Constant() {r=500}) {materialFunction=&CheapContrast}.Result) {r=true}
  n36 = Subtract(A=ComponentMask(None=@n30) {b=true}) {constB=0.5}
  n37 = Constant() {r=-2.0651}
  n38 = Constant() {r=1}
  n39 = ComponentMask(None=VertexNormalWS()) {b=true}
  n40 = LinearInterpolate(
   A=LinearInterpolate(
    A=TextureSample(UVs=@n31, Tex=@n32) {samplerSource=SSM_Wrap_WorldGroupSettings}.RGB
    B=TextureSample(UVs=@n33, Tex=@n32) {samplerSource=SSM_Wrap_WorldGroupSettings}.RGB
    Alpha=Saturate(None=@n35)
   ) {desc=Float3}
   B=Saturate(None=Divide(A=@n36) {constB=0.1})
   Alpha=Saturate(None=LinearInterpolate(A=@n37, B=@n38, Alpha=@n39) {constA=-0.3})
  )
  n41 = ComponentMask(None=@n40) {r=true}
  n42 = Multiply(A=NamedRerouteUsage(), B=@n41)
  n43 = FunctionInput(Preview=StaticBool() {value=true}) {bUsePreviewValueAsDefault=true, inputName=UseLayerMasking, inputType=FunctionInput_StaticBool, sortPriority=500}
  n44 = Round(None=Saturate(None=StaticSwitch(True=@n42, False=@n41, Value=@n43)))
  n45 = MaterialFunctionCall!(NoRVT=@n44, RVT=StaticSwitchParameter(True=Multiply(A=@n14, B=@n44), False=@n44) {group=Grass Blend, parameterName="Use RVT  GrassBlend"}) {materialFunction=/Game/Packages/Fortress_SharedMaterials/Base_Material_Functions/FN_RVT_Switch_Scalar}
  n46 = PixelDepth()
  n47 = Divide(A=Subtract(A=@n46.R) {constB=5000}) {constB=1000}
  n48 = ShadingPathSwitch(Default=@n45.Result, Mobile=LinearInterpolate(A=@n45.Result, Alpha=Saturate(None=@n47)) {constB=0}) {desc=Migrated from FeatureLevelSwitch}
  n49 = MaterialProxyReplace(
   Realtime=StaticSwitchParameter(True=BlendMaterialAttributes(A=@n1, B=@n12, Alpha=@n48), False=@n1) {group=Grass Blend, parameterName=UseGrassblend}
   MaterialProxy=@n1
  )
  FunctionOutput(None=@n49)
  FunctionOutput(None=StaticSwitchParameter(True=@n48, False=Constant()) {group=Grass Blend, parameterName=UseGrassblend}) {outputName=Grass Mask, sortPriority=1}
  n50 = ComponentMask(None=@n40) {g=true}
  n51 = LinearInterpolate(
   A=ScalarParameter() {defaultValue=-0.3496, group=Grass Blend, parameterName=GrassBlendOutLineThickness}
   B=ScalarParameter() {defaultValue=0.4584, group=Grass Blend, parameterName=GrassBlendOutLineThickness_Max}
   Alpha=Saturate(None=Divide(A=@n46.R) {constB=5000})
  )
  n52 = Subtract(A=@n46.R, B=ScalarParameter() {defaultValue=2000, parameterName=MobileGrassEdgeLine_Fade_Offset})
  n53 = ScalarParameter() {defaultValue=2000, parameterName=MobileGrassEdgeLine_Fade_Blend}
  LinearInterpolate(A=Round(None=Subtract(A=@n50, B=@n51)), Alpha=Saturate(None=Divide(A=@n52, B=@n53)))
  LinearInterpolate(
   A=ScalarParameter() {defaultValue=-50, group=Grass Blend, parameterName=GrassBlend_RVTMaskCliffs_Low}
   B=ScalarParameter() {defaultValue=5, group=Grass Blend, parameterName=GrassBlend_RVTMaskCliffs_High}
   Alpha=DotProduct(A=@n5, B=@n4.Grass BaseColor)
  )
  n54 = VectorParameter() {group=Grass Blend, parameterName=GrassBlendVisiblityMask, channelNames {a=Layer4, b=Layer3, g=Layer2, r=Layer1}, defaultValue=<rgba 1 0 0 0>}
  n55 = StaticSwitch(
   True=LinearInterpolate(A=@n54.Layer1, B=@n54.Layer1, Alpha=FunctionInput() {inputName=L2_Mask, inputType=FunctionInput_Scalar, sortPriority=101})
   False=@n54.Layer1
   Value=StaticBoolParameter() {group=2_Layer, parameterName=Use 2 Layers, sortPriority=-10}
  )
  n56 = StaticSwitch(
   True=LinearInterpolate(A=@n55, B=@n54.Layer3, Alpha=FunctionInput() {inputName=L3_Mask, inputType=FunctionInput_Scalar, sortPriority=102})
   False=@n55
   Value=StaticBoolParameter() {group=3_Layer, parameterName=Use 3 Layers, sortPriority=-10}
  )
  n57 = FunctionInput() {inputName=L4_Mask, inputType=FunctionInput_Scalar, sortPriority=103}
  NamedRerouteDeclaration(None=StaticSwitch(
    True=LinearInterpolate(A=@n56, B=@n54.Layer4, Alpha=@n57)
    False=@n56
    Value=StaticBoolParameter() {group=4_Layer, parameterName=Use 4 Layers, sortPriority=-10}
   )) {name=LayerMask, nodeColor=<rgba 0.8047 0 1 1>}
  n58 = Subtract(A=@n26, B=ScalarParameter() {defaultValue=16, group=Grass Blend, parameterName=GrassBlend_AutoBlend_Offset})
  n59 = ScalarParameter() {defaultValue=17, group=Grass Blend, parameterName=GrassBlend_AutoBlend_Distance}
  NamedRerouteDeclaration(None=Saturate(None=Divide(A=@n58, B=@n59))) {name=RVTFade, nodeColor=<rgba 0.9219 1 0 1>}
  OneMinus(None=@n50)
  n60 = Length(None=Subtract(A=@n5, B=@n4.Grass BaseColor)) {is=class MaterialExpressionLength}
  n61 = ScalarParameter() {defaultValue=0.025, group=Grass Blend, parameterName=GrassBlend_RVTMaskTreshold}
  OneMinus(None=Ceil(None=Subtract(A=@n60, B=@n61)))
  OneMinus(None=OneMinus(None=@n13))
  n62 = FunctionInput() {bUsePreviewValueAsDefault=true, inputName=RVT_Roughness, inputType=FunctionInput_Scalar, sortPriority=21}
  Saturate(None=Saturate(None=Floor(None=@n62)))
  Saturate(None=Ceil(None=Subtract(A=@n62, B=@n61)))
  ScalarParameter() {defaultValue=500, parameterName=GrassProjectionContrast}
  n63 = MaterialFunctionCall(Number=@n22, UVs=ScreenPosition().ViewportUV) {materialFunction=&DebugScalarValues}
  SetMaterialAttributes!(
   MaterialAttributes=@n49
   EmissiveColor=@n63.Result
   BaseColor=Add(A=GetMaterialAttributes(None=@n49) {attributeGetTypes [#1]}.BaseColor, B=@n63.Result)
  ) {attributeSetTypes [#5, #1]}
  StaticSwitchParameter(True=LinearInterpolate(A=@n14, Alpha=NamedRerouteUsage()), False=@n14) {group=Grass Blend, parameterName=UseAutoGrassblend}
  VectorParameter() {group=Grass Blend, parameterName=GrassColor, defaultValue=<rgba 0.022 0.1927 0.0451 1>}
 ]}
